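# Post-process style AnimBlueprint for the female medium Suspenders body, adding secondary motion to belt, rings, sleeves and cuffs.
# It copies the attached parent mesh's pose, then, per side, constrains belt and hip-pivot bones to thigh and pelvis before angular-spring AnimDynamics.
# Linear-spring sleeve and angular-spring cuff simulations follow, feeding the output.
# A separate linear-spring cuff chain is disconnected; the EventGraph is empty and the declared variables are unread.
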AnimBlueprint /Game/Characters/Player/Female/Medium/Bodies/F_MED_Suspenders/Meshes/F_MED_Suspenders_AnimBP
target_skeleton=&/F_MED_Suspenders_Skeleton
graphs [
 AnimGraph {nodes [
   n1 = Anim Dynamics - Bone: dyn_L_cuff_1(Alpha=1, ExternalForce="0,0,0") {
    bGravityOverrideInSimSpace=true
    bLinearSpring=true
    gravityScale=0.1
    linearDampingOverride=0.9999
    linearSpringConstant=10
    simulationSpace=RootRelative
    alphaBoolBlend {blendOption=Linear}
    boundBone {boneName=dyn_L_cuff_1}
    componentLinearAccScale=<1 1 1>
    physicsBodyDefinitions [{
      sphereCollisionRadius=0
      boundBone {boneName=dyn_L_cuff_1}
      boxExtents=<1 1 1>
      constraintSetup {angularLimitsMax=<0 0 0>, angularLimitsMin=<0 0 0>, angularTarget=<0 0 0>, linearAxesMax=<3 2 0>, linearAxesMin=<-0.7 -2 0>}
      localJointOffset=<0 0 0>
     }]
   }
   n2 = Anim Dynamics - Bone: dyn_L_cuff_2(ComponentPose=@n1, Alpha=1, ExternalForce="0,0,0") {
    bGravityOverrideInSimSpace=true
    bLinearSpring=true
    gravityScale=0.1
    linearDampingOverride=0.9999
    linearSpringConstant=10
    simulationSpace=RootRelative
    alphaBoolBlend {blendOption=Linear}
    boundBone {boneName=dyn_L_cuff_2}
    componentLinearAccScale=<1 1 1>
    physicsBodyDefinitions [{
      sphereCollisionRadius=0
      boundBone {boneName=dyn_L_cuff_2}
      boxExtents=<1 1 1>
      constraintSetup {angularLimitsMax=<0 0 0>, angularLimitsMin=<0 0 0>, angularTarget=<0 0 0>, linearAxesMax=<3 2 0>, linearAxesMin=<-0.7 -2 0>}
      localJointOffset=<0 0 0>
     }]
   }
   n3 = Anim Dynamics - Bone: dyn_L_cuff_3(ComponentPose=@n2, Alpha=1, ExternalForce="0,0,0") {
    bGravityOverrideInSimSpace=true
    bLinearSpring=true
    gravityScale=0.1
    linearDampingOverride=0.9999
    linearSpringConstant=10
    simulationSpace=RootRelative
    alphaBoolBlend {blendOption=Linear}
    boundBone {boneName=dyn_L_cuff_3}
    componentLinearAccScale=<1 1 1>
    physicsBodyDefinitions [{
      sphereCollisionRadius=0
      boundBone {boneName=dyn_L_cuff_3}
      boxExtents=<1 1 1>
      constraintSetup {angularLimitsMax=<0 0 0>, angularLimitsMin=<0 0 0>, angularTarget=<0 0 0>, linearAxesMax=<3 2 0>, linearAxesMin=<-0.7 -2 0>}
      localJointOffset=<0 0 0>
     }]
   }
   n4 = Anim Dynamics - Bone: dyn_R_cuff_1(ComponentPose=@n3, Alpha=1, ExternalForce="0,0,0") {
    bGravityOverrideInSimSpace=true
    bLinearSpring=true
    gravityScale=0.1
    linearDampingOverride=0.9999
    linearSpringConstant=10
    simulationSpace=RootRelative
    alphaBoolBlend {blendOption=Linear}
    boundBone {boneName=dyn_R_cuff_1}
    componentLinearAccScale=<1 1 1>
    physicsBodyDefinitions [{
      sphereCollisionRadius=0
      boundBone {boneName=dyn_R_cuff_1}
      boxExtents=<1 1 1>
      constraintSetup {angularLimitsMax=<0 0 0>, angularLimitsMin=<0 0 0>, angularTarget=<0 0 0>, linearAxesMax=<3 2 0>, linearAxesMin=<-0.7 -2 0>}
      localJointOffset=<0 0 0>
     }]
   }
   n5 = Anim Dynamics - Bone: dyn_R_cuff_2(ComponentPose=@n4, Alpha=1, ExternalForce="0,0,0") {
    bGravityOverrideInSimSpace=true
    bLinearSpring=true
    gravityScale=0.1
    linearDampingOverride=0.9999
    linearSpringConstant=10
    simulationSpace=RootRelative
    alphaBoolBlend {blendOption=Linear}
    boundBone {boneName=dyn_R_cuff_2}
    componentLinearAccScale=<1 1 1>
    physicsBodyDefinitions [{
      sphereCollisionRadius=0
      boundBone {boneName=dyn_R_cuff_2}
      boxExtents=<1 1 1>
      constraintSetup {angularLimitsMax=<0 0 0>, angularLimitsMin=<0 0 0>, angularTarget=<0 0 0>, linearAxesMax=<3 2 0>, linearAxesMin=<-0.7 -2 0>}
      localJointOffset=<0 0 0>
     }]
   }
   Anim Dynamics - Bone: dyn_R_cuff_3(ComponentPose=@n5, Alpha=1, ExternalForce="0,0,0") {
    bGravityOverrideInSimSpace=true
    bLinearSpring=true
    gravityScale=0.1
    linearDampingOverride=0.9999
    linearSpringConstant=10
    simulationSpace=RootRelative
    alphaBoolBlend {blendOption=Linear}
    boundBone {boneName=dyn_R_cuff_3}
    componentLinearAccScale=<1 1 1>
    physicsBodyDefinitions [{
      sphereCollisionRadius=0
      boundBone {boneName=dyn_R_cuff_3}
      boxExtents=<1 1 1>
      constraintSetup {angularLimitsMax=<0 0 0>, angularLimitsMin=<0 0 0>, angularTarget=<0 0 0>, linearAxesMax=<3 2 0>, linearAxesMin=<-0.7 -2 0>}
      localJointOffset=<0 0 0>
     }]
   }
   n6 = "Constraint - dyn_L_beltRing_front (R)"(ComponentPose=Local To Component(LocalPose=Copy Pose From Mesh() {bUseAttachedParent=true}), Alpha=1, ConstraintWeights_0=1) {
    alphaBoolBlend {blendOption=Linear}
    boneToModify {boneName=dyn_L_beltRing_front}
    constraintSetup [{transformType=Rotation, targetBone {boneName=thigh_twist_01_l}}]
    constraintWeights [1]
   }
   n7 = "Constraint - dyn_L_belt_B (PP)"!(ComponentPose=@n6, Alpha=1, ConstraintWeights_0=1, ConstraintWeights_1:float=0) -> (Pose) {
    alphaBoolBlend {blendOption=Linear}
    boneToModify {boneName=dyn_L_belt_B}
    constraintSetup [{targetBone {boneName=thigh_twist_01_l}}, {targetBone {boneName=pelvis}}]
    constraintWeights [1,1]
   }
   n8 = "Constraint - dyn_L_hip_pivot_D_modify (PPP)"!(ComponentPose=@n7, Alpha=1, ConstraintWeights_0=0.2, ConstraintWeights_1:float=0, ConstraintWeights_2:float=0.8) -> (Pose) {
    alphaBoolBlend {blendOption=Linear}
    boneToModify {boneName=dyn_L_hip_pivot_D_modify}
    constraintSetup [{targetBone {boneName=thigh_l}}, {targetBone {boneName=thigh_twist_01_l}}, {targetBone {boneName=pelvis}}]
    constraintWeights [1,1,1]
   }
   n9 = "Constraint - dyn_L_belt_D (PPP)"!(ComponentPose=@n8, Alpha=1, ConstraintWeights_0=0.632, ConstraintWeights_1:float=0.0266, ConstraintWeights_2:float=0.3414) -> (Pose) {
    alphaBoolBlend {blendOption=Linear}
    boneToModify {boneName=dyn_L_belt_D}
    constraintSetup [{targetBone {boneName=thigh_l}}, {targetBone {boneName=thigh_twist_01_l}}, {targetBone {boneName=pelvis}}]
    constraintWeights [1,1,1]
   }
   n10 = Anim Dynamics - Bone: dyn_L_beltRing_front(ComponentPose=@n9, Alpha=1, ExternalForce="0,0,0") {
    angularDampingOverride=1
    angularSpringConstant=20000
    bAngularSpring=true
    bOverrideAngularDamping=true
    bUseGravityOverride=true
    simulationSpace=World
    alphaBoolBlend {blendOption=Linear}
    boundBone {boneName=dyn_L_beltRing_front}
    componentLinearAccScale=<1 1 1>
    componentLinearVelScale=<1 1 1>
    gravityOverride {z=-500}
    physicsBodyDefinitions [{
      boundBone {boneName=dyn_L_beltRing_front}
      boxExtents=<36 36 36>
      constraintSetup {angularTargetAxis=AxisZ, angularLimitsMax=<0 0 0>, angularLimitsMin=<0 -70 0>, angularTarget=<0 0 1>}
      localJointOffset=<0 0 -3>
     }]
   }
   n11 = Anim Dynamics - Bone: dyn_L_belt_B(ComponentPose=@n10, Alpha=1, ExternalForce="0,0,0") {
    angularDampingOverride=1
    angularSpringConstant=20000
    bAngularSpring=true
    bOverrideAngularDamping=true
    bUseGravityOverride=true
    simulationSpace=World
    alphaBoolBlend {blendOption=Linear}
    boundBone {boneName=dyn_L_belt_B}
    componentLinearAccScale=<1 1 1>
    componentLinearVelScale=<1 1 1>
    gravityOverride {z=-500}
    physicsBodyDefinitions [{
      boundBone {boneName=dyn_L_belt_B}
      boxExtents=<36 36 36>
      constraintSetup {angularLimitsMax=<0 0 0>, angularLimitsMin=<0 -70 0>, angularTarget=<1 0 0>}
      localJointOffset=<-3 0 0>
     }]
   }
   n12 = Anim Dynamics - Bone: dyn_L_hip_pivot_D(ComponentPose=@n11, Alpha=1, ExternalForce="0,0,0") {
    angularDampingOverride=1
    angularSpringConstant=20000
    bAngularSpring=true
    bOverrideAngularDamping=true
    bUseGravityOverride=true
    simulationSpace=World
    alphaBoolBlend {blendOption=Linear}
    boundBone {boneName=dyn_L_hip_pivot_D}
    componentLinearAccScale=<1 1 1>
    componentLinearVelScale=<1 1 1>
    gravityOverride {z=-500}
    physicsBodyDefinitions [{
      boundBone {boneName=dyn_L_hip_pivot_D}
      boxExtents=<36 36 36>
      constraintSetup {angularLimitsMax=<0 70 0>, angularLimitsMin=<0 0 0>, angularTarget=<1 0 0>}
      localJointOffset=<-3 0 0>
     }]
   }
   n13 = Anim Dynamics - Bone: dyn_L_hip_pivot_D_modify(ComponentPose=@n12, Alpha=1, ExternalForce="0,0,0") {
    angularDampingOverride=1
    angularSpringConstant=20000
    bAngularSpring=true
    bOverrideAngularDamping=true
    bUseGravityOverride=true
    simulationSpace=World
    alphaBoolBlend {blendOption=Linear}
    boundBone {boneName=dyn_L_hip_pivot_D_modify}
    componentLinearAccScale=<1 1 1>
    componentLinearVelScale=<1 1 1>
    gravityOverride {z=-500}
    physicsBodyDefinitions [{
      boundBone {boneName=dyn_L_hip_pivot_D_modify}
      boxExtents=<36 36 36>
      constraintSetup {angularLimitsMax=<0 70 0>, angularLimitsMin=<0 0 0>, angularTarget=<1 0 0>}
      localJointOffset=<-3 0 0>
     }]
   }
   n14 = Anim Dynamics - Bone: dyn_L_beltRing_back(ComponentPose=@n13, Alpha=1, ExternalForce="0,0,0") {
    angularDampingOverride=1
    angularSpringConstant=20000
    bAngularSpring=true
    bOverrideAngularDamping=true
    bUseGravityOverride=true
    simulationSpace=World
    alphaBoolBlend {blendOption=Linear}
    boundBone {boneName=dyn_L_beltRing_back}
    componentLinearAccScale=<1 1 1>
    componentLinearVelScale=<1 1 1>
    gravityOverride {z=-500}
    physicsBodyDefinitions [{
      boundBone {boneName=dyn_L_beltRing_back}
      boxExtents=<36 36 36>
      constraintSetup {angularTargetAxis=AxisZ, angularLimitsMax=<0 0 0>, angularLimitsMin=<-45 -70 0>, angularTarget=<0 0 1>}
      localJointOffset=<0 0 -3>
     }]
   }
   n15 = "Constraint - dyn_R_beltRing_front (R)"(ComponentPose=@n14, Alpha=1, ConstraintWeights_0=1) {
    alphaBoolBlend {blendOption=Linear}
    boneToModify {boneName=dyn_R_beltRing_front}
    constraintSetup [{transformType=Rotation, targetBone {boneName=thigh_twist_01_r}}]
    constraintWeights [1]
   }
   n16 = "Constraint - dyn_R_belt_B (PP)"!(ComponentPose=@n15, Alpha=1, ConstraintWeights_0=1, ConstraintWeights_1:float=0) -> (Pose) {
    alphaBoolBlend {blendOption=Linear}
    boneToModify {boneName=dyn_R_belt_B}
    constraintSetup [{targetBone {boneName=thigh_twist_01_r}}, {targetBone {boneName=pelvis}}]
    constraintWeights [1,1]
   }
   n17 = "Constraint - dyn_R_hip_pivot_D_modify (PPP)"!(ComponentPose=@n16, Alpha=1, ConstraintWeights_0=0.2, ConstraintWeights_1:float=0, ConstraintWeights_2:float=0.8) -> (Pose) {
    alphaBoolBlend {blendOption=Linear}
    boneToModify {boneName=dyn_R_hip_pivot_D_modify}
    constraintSetup [{targetBone {boneName=thigh_r}}, {targetBone {boneName=thigh_twist_01_r}}, {targetBone {boneName=pelvis}}]
    constraintWeights [1,1,1]
   }
   n18 = "Constraint - dyn_R_belt_D (PPP)"!(ComponentPose=@n17, Alpha=1, ConstraintWeights_0=0.632, ConstraintWeights_1:float=0.0266, ConstraintWeights_2:float=0.3414) -> (Pose) {
    alphaBoolBlend {blendOption=Linear}
    boneToModify {boneName=dyn_R_belt_D}
    constraintSetup [{targetBone {boneName=thigh_r}}, {targetBone {boneName=thigh_twist_01_r}}, {targetBone {boneName=pelvis}}]
    constraintWeights [1,1,1]
   }
   n19 = Anim Dynamics - Bone: dyn_R_beltRing_front(ComponentPose=@n18, Alpha=1, ExternalForce="0,0,0") {
    angularDampingOverride=1
    angularSpringConstant=20000
    bAngularSpring=true
    bOverrideAngularDamping=true
    bUseGravityOverride=true
    simulationSpace=World
    alphaBoolBlend {blendOption=Linear}
    boundBone {boneName=dyn_R_beltRing_front}
    componentLinearAccScale=<1 1 1>
    componentLinearVelScale=<1 1 1>
    gravityOverride {z=-500}
    physicsBodyDefinitions [{
      boundBone {boneName=dyn_R_beltRing_front}
      boxExtents=<36 36 36>
      constraintSetup {angularTargetAxis=AxisZ, angularLimitsMax=<0 0 0>, angularLimitsMin=<0 -70 0>, angularTarget=<0 0 1>}
      localJointOffset=<0 0 -3>
     }]
   }
   n20 = Anim Dynamics - Bone: dyn_R_belt_B(ComponentPose=@n19, Alpha=1, ExternalForce="0,0,0") {
    angularDampingOverride=1
    angularSpringConstant=20000
    bAngularSpring=true
    bOverrideAngularDamping=true
    bUseGravityOverride=true
    simulationSpace=World
    alphaBoolBlend {blendOption=Linear}
    boundBone {boneName=dyn_R_belt_B}
    componentLinearAccScale=<1 1 1>
    componentLinearVelScale=<1 1 1>
    gravityOverride {z=-500}
    physicsBodyDefinitions [{
      boundBone {boneName=dyn_R_belt_B}
      boxExtents=<36 36 36>
      constraintSetup {angularLimitsMax=<0 0 0>, angularLimitsMin=<0 -70 0>, angularTarget=<1 0 0>}
      localJointOffset=<3 0 0>
     }]
   }
   n21 = Anim Dynamics - Bone: dyn_R_hip_pivot_D(ComponentPose=@n20, Alpha=1, ExternalForce="0,0,0") {
    angularDampingOverride=1
    angularSpringConstant=20000
    bAngularSpring=true
    bOverrideAngularDamping=true
    bUseGravityOverride=true
    simulationSpace=World
    alphaBoolBlend {blendOption=Linear}
    boundBone {boneName=dyn_R_hip_pivot_D}
    componentLinearAccScale=<1 1 1>
    componentLinearVelScale=<1 1 1>
    gravityOverride {z=-500}
    physicsBodyDefinitions [{
      boundBone {boneName=dyn_R_hip_pivot_D}
      boxExtents=<36 36 36>
      constraintSetup {angularLimitsMax=<0 70 0>, angularLimitsMin=<0 0 0>, angularTarget=<1 0 0>}
      localJointOffset=<3 0 0>
     }]
   }
   n22 = Anim Dynamics - Bone: dyn_R_hip_pivot_D_modify(ComponentPose=@n21, Alpha=1, ExternalForce="0,0,0") {
    angularDampingOverride=1
    angularSpringConstant=20000
    bAngularSpring=true
    bOverrideAngularDamping=true
    bUseGravityOverride=true
    simulationSpace=World
    alphaBoolBlend {blendOption=Linear}
    boundBone {boneName=dyn_R_hip_pivot_D_modify}
    componentLinearAccScale=<1 1 1>
    componentLinearVelScale=<1 1 1>
    gravityOverride {z=-500}
    physicsBodyDefinitions [{
      boundBone {boneName=dyn_R_hip_pivot_D_modify}
      boxExtents=<36 36 36>
      constraintSetup {angularLimitsMax=<0 70 0>, angularLimitsMin=<0 0 0>, angularTarget=<1 0 0>}
      localJointOffset=<3 0 0>
     }]
   }
   n23 = Anim Dynamics - Bone: dyn_R_beltRing_back(ComponentPose=@n22, Alpha=1, ExternalForce="0,0,0") {
    angularDampingOverride=1
    angularSpringConstant=20000
    bAngularSpring=true
    bOverrideAngularDamping=true
    bUseGravityOverride=true
    simulationSpace=World
    alphaBoolBlend {blendOption=Linear}
    boundBone {boneName=dyn_R_beltRing_back}
    componentLinearAccScale=<1 1 1>
    componentLinearVelScale=<1 1 1>
    gravityOverride {z=-500}
    physicsBodyDefinitions [{
      boundBone {boneName=dyn_R_beltRing_back}
      boxExtents=<36 36 36>
      constraintSetup {angularTargetAxis=AxisZ, angularLimitsMax=<45 0 0>, angularLimitsMin=<0 -70 0>, angularTarget=<0 0 1>}
      localJointOffset=<0 0 -3>
     }]
   }
   n24 = Anim Dynamics - Bone: dyn_L_sleeve(ComponentPose=@n23, Alpha=1, ExternalForce="0,0,0") {
    bGravityOverrideInSimSpace=true
    bLinearSpring=true
    bOverrideLinearDamping=true
    gravityScale=0.5
    linearDampingOverride=0.9999
    linearSpringConstant=100
    simulationSpace=RootRelative
    alphaBoolBlend {blendOption=Linear}
    alphaScaleBiasClamp {scale=0.5}
    boundBone {boneName=dyn_L_sleeve}
    componentLinearAccScale=<1 1 1>
    componentLinearVelScale=<0.5 0.5 0.5>
    physicsBodyDefinitions [{
      sphereCollisionRadius=0
      boundBone {boneName=dyn_L_sleeve}
      boxExtents=<6 6 1>
      constraintSetup {angularLimitsMax=<0 0 0>, angularLimitsMin=<0 0 0>, angularTarget=<0 0 0>, linearAxesMax=<1 2 0.5>, linearAxesMin=<-1 -2 0>}
      localJointOffset=<0 0 0>
     }]
   }
   n25 = Anim Dynamics - Bone: dyn_R_sleeve(ComponentPose=@n24, Alpha=1, ExternalForce="0,0,0") {
    bGravityOverrideInSimSpace=true
    bLinearSpring=true
    bOverrideLinearDamping=true
    gravityScale=0.5
    linearDampingOverride=0.9999
    linearSpringConstant=100
    simulationSpace=RootRelative
    alphaBoolBlend {blendOption=Linear}
    alphaScaleBiasClamp {scale=0.5}
    boundBone {boneName=dyn_R_sleeve}
    componentLinearAccScale=<1 1 1>
    componentLinearVelScale=<0.5 0.5 0.5>
    physicsBodyDefinitions [{
      sphereCollisionRadius=0
      boundBone {boneName=dyn_R_sleeve}
      boxExtents=<6 6 1>
      constraintSetup {angularLimitsMax=<0 0 0>, angularLimitsMin=<0 0 0>, angularTarget=<0 0 0>, linearAxesMax=<1 2 0.5>, linearAxesMin=<-1 -2 0>}
      localJointOffset=<0 0 0>
     }]
   }
   n26 = Anim Dynamics - Bone: dyn_L_cuff_1(ComponentPose=@n25, Alpha=1, ExternalForce="0,0,0") {
    angularDampingOverride=0.9999
    angularSpringConstant=2000
    bAngularSpring=true
    bGravityOverrideInSimSpace=true
    bOverrideAngularDamping=true
    bUseGravityOverride=true
    gravityScale=0.1
    linearDampingOverride=0.9999
    linearSpringConstant=10
    simulationSpace=World
    alphaBoolBlend {blendOption=Linear}
    boundBone {boneName=dyn_L_cuff_1}
    componentLinearAccScale=<1 1 1>
    gravityOverride {z=-980}
    physicsBodyDefinitions [{
      sphereCollisionRadius=0
      boundBone {boneName=dyn_L_cuff_1}
      boxExtents=<12 12 12>
      constraintSetup {angularLimitsMax=<30 0 30>, angularLimitsMin=<-30 0 -30>, angularTarget=<1 0 0>}
      localJointOffset=<1 0 0>
     }]
   }
   n27 = Anim Dynamics - Bone: dyn_L_cuff_2(ComponentPose=@n26, Alpha=1, ExternalForce="0,0,0") {
    angularDampingOverride=0.9999
    angularSpringConstant=2000
    bAngularSpring=true
    bGravityOverrideInSimSpace=true
    bOverrideAngularDamping=true
    bUseGravityOverride=true
    gravityScale=0.1
    linearDampingOverride=0.9999
    linearSpringConstant=10
    simulationSpace=World
    alphaBoolBlend {blendOption=Linear}
    boundBone {boneName=dyn_L_cuff_2}
    componentLinearAccScale=<1 1 1>
    gravityOverride {z=-980}
    physicsBodyDefinitions [{
      sphereCollisionRadius=0
      boundBone {boneName=dyn_L_cuff_2}
      boxExtents=<12 12 12>
      constraintSetup {angularLimitsMax=<30 0 30>, angularLimitsMin=<-30 0 -30>, angularTarget=<1 0 0>}
      localJointOffset=<1 0 0>
     }]
   }
   n28 = Anim Dynamics - Bone: dyn_L_cuff_3(ComponentPose=@n27, Alpha=1, ExternalForce="0,0,0") {
    angularDampingOverride=0.9999
    angularSpringConstant=2000
    bAngularSpring=true
    bGravityOverrideInSimSpace=true
    bOverrideAngularDamping=true
    bUseGravityOverride=true
    gravityScale=0.1
    linearDampingOverride=0.9999
    linearSpringConstant=10
    simulationSpace=World
    alphaBoolBlend {blendOption=Linear}
    boundBone {boneName=dyn_L_cuff_3}
    componentLinearAccScale=<1 1 1>
    gravityOverride {z=-980}
    physicsBodyDefinitions [{
      sphereCollisionRadius=0
      boundBone {boneName=dyn_L_cuff_3}
      boxExtents=<12 12 12>
      constraintSetup {angularLimitsMax=<30 0 30>, angularLimitsMin=<-30 0 -30>, angularTarget=<1 0 0>}
      localJointOffset=<1 0 0>
     }]
   }
   n29 = Anim Dynamics - Bone: dyn_R_cuff_1(ComponentPose=@n28, Alpha=1, ExternalForce="0,0,0") {
    angularDampingOverride=0.9999
    angularSpringConstant=2000
    bAngularSpring=true
    bGravityOverrideInSimSpace=true
    bOverrideAngularDamping=true
    bUseGravityOverride=true
    gravityScale=0.1
    linearDampingOverride=0.9999
    linearSpringConstant=10
    simulationSpace=World
    alphaBoolBlend {blendOption=Linear}
    boundBone {boneName=dyn_R_cuff_1}
    componentLinearAccScale=<1 1 1>
    gravityOverride {z=-980}
    physicsBodyDefinitions [{
      sphereCollisionRadius=0
      boundBone {boneName=dyn_R_cuff_1}
      boxExtents=<12 12 12>
      constraintSetup {angularLimitsMax=<30 0 30>, angularLimitsMin=<-30 0 -30>, angularTarget=<1 0 0>}
      localJointOffset=<1 0 0>
     }]
   }
   n30 = Anim Dynamics - Bone: dyn_R_cuff_2(ComponentPose=@n29, Alpha=1, ExternalForce="0,0,0") {
    angularDampingOverride=0.9999
    angularSpringConstant=2000
    bAngularSpring=true
    bGravityOverrideInSimSpace=true
    bOverrideAngularDamping=true
    bUseGravityOverride=true
    gravityScale=0.1
    linearDampingOverride=0.9999
    linearSpringConstant=10
    simulationSpace=World
    alphaBoolBlend {blendOption=Linear}
    boundBone {boneName=dyn_R_cuff_2}
    componentLinearAccScale=<1 1 1>
    gravityOverride {z=-980}
    physicsBodyDefinitions [{
      sphereCollisionRadius=0
      boundBone {boneName=dyn_R_cuff_2}
      boxExtents=<12 12 12>
      constraintSetup {angularLimitsMax=<30 0 30>, angularLimitsMin=<-30 0 -30>, angularTarget=<1 0 0>}
      localJointOffset=<1 0 0>
     }]
   }
   n31 = Anim Dynamics - Bone: dyn_R_cuff_3(ComponentPose=@n30, Alpha=1, ExternalForce="0,0,0") {
    angularDampingOverride=0.9999
    angularSpringConstant=2000
    bAngularSpring=true
    bGravityOverrideInSimSpace=true
    bOverrideAngularDamping=true
    bUseGravityOverride=true
    gravityScale=0.1
    linearDampingOverride=0.9999
    linearSpringConstant=10
    simulationSpace=World
    alphaBoolBlend {blendOption=Linear}
    boundBone {boneName=dyn_R_cuff_3}
    componentLinearAccScale=<1 1 1>
    gravityOverride {z=-980}
    physicsBodyDefinitions [{
      sphereCollisionRadius=0
      boundBone {boneName=dyn_R_cuff_3}
      boxExtents=<12 12 12>
      constraintSetup {angularLimitsMax=<30 0 30>, angularLimitsMin=<-30 0 -30>, angularTarget=<1 0 0>}
      localJointOffset=<1 0 0>
     }]
   }
   Output Pose(Result=Component To Local(ComponentPose=@n31)) {layerGroup=None}
  ]}
 EventGraph {links [], nodes []}
]
variables [
 Is Accelerating 2D {type=number}
 Is Playing Emote {type=number}
 isCrouching {type=number}
 RBANoff {type=number}
]
